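# Animation Blueprint for the M_MED_AgileRug male player body that copies the attached parent mesh's pose and refines it through the body's Control Rig.
# Each update derives rig inputs for skydiving, emotes, front end and pistol aiming.
# The pistol value updates only in the Pistol case of the parent FortPlayerAnimInstance's weapon type.
# A shoulder-pad fixed-position value interpolates toward on while emoting or skydiving.
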AnimBlueprint /BRCosmetics/Characters/Player/Male/Medium/Bodies/M_MED_AgileRug/Meshes/M_MED_AgileRug_AnimBP
target_skeleton=&/M_MED_AgileRug_Skeleton
graphs [
 AnimGraph {nodes [
   n1 = Control Rig!(
    Source=Copy Pose From Mesh() {bUseAttachedParent=true}
    Alpha
    isSkydiving:bool=Get bIsSkyDiving()
    IsEmote:bool=Get bIsPlayingEmote()
    isFE:bool=Get Is FE() -> (Is FE:bool)
    IsPistol:double=Get Pistol() -> (Pistol:double)
    ShoulderPadsFixedPosition:double=Get ShoulderPadsFixedPosition() -> (ShoulderPadsFixedPosition:double)
   ) -> (Pose) {
    alphaBoolBlend {blendOption=Linear}
    controlRigAssetReference {blueprintRigClass=&/M_MED_AgileRug_CtrlRig.M_MED_AgileRug_CtrlRig_C}
   }
   Output Pose(Result=@n1) {layerGroup=None}
  ]}
 EventGraph {nodes [
   n1 = Event BlueprintUpdateAnimation()
   > Set Delta_Time_X(Delta_Time_X:double=@n1.DeltaTimeX) -> (Output_Get:double)
   > n2 = Sequence!(execute) -> (then_0, then_1, then_2:Exec)
   n3 = Get bIsPlayingEmote()
   > Set Is FE(Is FE:bool=AND Boolean(A=Get bIsInFrontEnd(), B=NOT Boolean(A=@n3))) -> (Output_Get:bool)
   n4 = OR Boolean(A=Get bIsPlayingEmote(), B=Get bIsSkyDiving())
   Set ShoulderPadsFixedPosition(
    execute=@n2.then_2
    ShoulderPadsFixedPosition:double=FInterpTo(
     Current=Get ShoulderPadsFixedPosition()
     Target="To Float (Boolean)"(InBool=@n4)
     DeltaTime=Get Delta_Time_X() -> (Delta_Time_X:double)
     InterpSpeed=5
    )
   ) -> (Output_Get:double)
   n5 = Cast To FortPlayerAnimInstance(execute=@n2.then_1, Object=GetAnimInstance(self=GetOwnerSkeletalMeshComponent()))
   > Set ar_firingPosition_GC(ar_firingPosition_GC:int=0) -> (Output_Get:int=0)
   > n6 = Switch on EFortWeaponCoreAnimation!(execute, Selection:EFortWeaponCoreAnimation Enum=Get WeaponCoreAnim(self=@n5)) -> (
    Melee:Exec
    Pistol:Exec
    Shotgun:Exec
    PaperBlueprint:Exec
    Rifle:Exec
    MeleeOneHand:Exec
    MachinePistol:Exec
    RocketLauncher:Exec
    GrenadeLauncher:Exec
    GoingCommando:Exec
    AssaultRifle:Exec
    TacticalShotgun:Exec
    SniperRifle:Exec
    TrapPlacement:Exec
    ShoulderLauncher:Exec
    AbilityDecoTool:Exec
    Crossbow:Exec
    C4:Exec
    RemoteControl:Exec
    DualWield:Exec
    AR_BullPup:Exec
    AR_ForwardGrip:Exec
    MedPackPaddles:Exec
    SMG_P90:Exec
    AR_DrumGun:Exec
    Consumable_Small:Exec
    Consumable_Large:Exec
    Balloon:Exec
    MountedTurret:Exec
    CreativeTool:Exec
    ExplosiveBow:Exec
    AshtonIndigo:Exec
    AshtonChicago:Exec
    MeleeDualWield:Exec
    Unarmed:Exec
   )
   n7 = Branch(execute=@n6.Pistol)
   > n8 = Set ar_firingPosition_GC(ar_firingPosition_GC=1) -> (Output_Get=0)
   > n9 = Set Gun_out(Gun_out:bool="To Boolean (Integer)"!(InInt:int=@n8) -> (ReturnValue:bool)) -> (Output_Get:bool)
   n10 = Get bIsTargeting()
   > Set Pistol(Pistol:double="To Float (Boolean)"(InBool=AND Boolean(A=@n9, B=@n10))) -> (Output_Get:double)
   Set ar_firingPosition_GC(execute=@n7.else, ar_firingPosition_GC=0) -> (Output_Get=0)
  ]}
]
variables [
 ar_firingPosition_GC {type=integer}
 Delta_Time_X {type=number}
 Gun_out {type=boolean}
 Is FE {type=boolean}
 Pistol {type=number}
 ShoulderPadsFixedPosition {type=number}
]
